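# Niagara system that adds a red glow and light to the Typhoon Robot backpack cosmetic.
# A single Flare emitter bursts one persistent particle at a fixed offset, feeding two sprite renderers and a light renderer.
# Custom particle attributes define separate soft-glow and light properties alongside the elongated, aligned flare.
# Time-of-day scalars adjust glow alpha, while random scaling varies flare size and dynamic parameters drive the material.
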NiagaraSystem /BRCosmetics/Accessories/FORT_Backpacks/Backpack_M_MED_Typhoon_Robot/FX/NS_Backpack_M_Typhoon_Robot_glow
dependencies {usedDataInterfaces [], usedDynamicInputs [], usedModules [&SystemState]}
emitters {Flare {
  emitterUpdateScript [
   &EmitterState(
    Inactive Response=::NewEnumerator1
    Life Cycle Mode=::NewEnumerator1
    Loop Behavior=::NewEnumerator1
    Loop Duration Mode=::NewEnumerator1
    UseLoopDelay=false
   )
   &SpawnBurst_Instantaneous(Spawn Count=1, Use Spawn Probability=false)
  ]
  particleSpawnScript [
   /Niagara/Modules/Spawn/Initialization/InitializeParticle(
    Color=<rgba 1.4 0.0365 0.0365 1>
    Module.Write Color=true
    Module.Write Lifetime=true
    Module.Write Mass=true
    Module.Write Material Random=true
    Module.Write Position=false
    Module.Write Scale=false
    Module.Write SpriteRotation=true
    Module.Write SpriteSize=true
    Position=&SimulationPosition()
    Sprite Rotation=90
    Sprite Size=<32 3.5>
   )
   &SystemLocation(MaskSpawnGroup=false, Offset=<0 -18.2 -1.2>)
   /SetVariables(Particles.Light_position:Vector3f=<0 -20 -0.5>) {isSetParameters=true}
   /SetVariables(Particles.Sprit_glow_soft_position:Vector3f=<0 -18.2 -1.2>) {isSetParameters=true}
  ]
  particleUpdateScript [
   /SetVariables(Particles.LightRadius=3) {isSetParameters=true}
   /SetVariables(Particles.Light_color:LinearColor=<rgba 10 0 0 1>) {isSetParameters=true}
   /SetVariables(Particles.Sprite_glow_soft_size:Vector2f=<8.5 8.5>) {isSetParameters=true}
   /SetVariables(Particles.Sprite_glow_soft_color:LinearColor=&MakeLinearColorFromVectorAndFloat(
     "Vector (RGB)"=&MakeVectorFromLinearColorRGB(Color=<rgba 1.4 0.0365 0.0365 1>)
     "Float (Alpha)"=&Multiply_Float(A=1, B=&FNTimeOfDayScalar(Night Scale=0.3, Morning Scale=0.4, Evening Scale=0.4))
    )) {isSetParameters=true}
   /SetVariables(Particles.Sprite_glow_Camera_offset_soft=-0.5) {isSetParameters=true}
   &DynamicMaterialParameters(
    Index 0 Param 1=3
    Index 0 Param 2=0.5
    Index 0 Param 3=0.2
    Index 0 Param 4=2
    Index 1 Param 2=0.2
    Index 1 Param 3=2
    Index 1 Param 4=1
    Module.Param0WriteEnabled=true
    Module.Param0WriteEnabled1=true
    Module.Param1WriteEnabled=true
    Module.Param1WriteEnabled1=true
    Module.Param2WriteEnabled=true
    Module.Param2WriteEnabled1=true
    Module.Param3WriteEnabled=true
    Module.Param3WriteEnabled1=true
    Write Parameter Index 1=true
   )
   &ParticleState(Kill Particles When Lifetime Has Elapsed=false, Let Infinitely Lived Particles Die When Emitter Deactivates=true)
   &SolveForcesAndVelocity(Clamp Velocity=false, Limit Acceleration=false)
   /NiagaraFN/Sprite/FN_SpriteAlignmentAndFacing(SpriteAlignment:Vector3f=<1 0 0>)
   &ScaleSpriteSize(Scale Factor=&Multiply_Vector2(
     A=<1 1>
     B=&Vector2DFromFloat(Value=&RandomRangeFloat(Minimum=0.8, Maximum=1, Evaluation Type=::NewEnumerator1))
    ))
   &CameraOffset(Camera Offset Amount=1.5)
   &ScaleColor(
    Scale Alpha=&Multiply_Float(A=1, B=&FNTimeOfDayScalar(Night Scale=0.5, Morning Scale=0.4, Evening Scale=0.4))
    ScaleA=true
    ScaleRGB=true
   )
  ]
  rendererClasses [&NiagaraSpriteRendererProperties, &NiagaraLightRendererProperties]
  renderers [
   {rendererClass=&NiagaraSpriteRendererProperties}
   {rendererClass=&NiagaraLightRendererProperties, rendererIndex=1}
   {rendererClass=&NiagaraSpriteRendererProperties, rendererIndex=2}
  ]
 }}
summary {
 systemName=NS_Backpack_M_Typhoon_Robot_glow
 emitters [{emitterName=Flare, rendererClasses [&NiagaraSpriteRendererProperties, &NiagaraLightRendererProperties]}]
 userVariables [User.SK_Mesh {
   type=NiagaraDataInterfaceSkeletalMesh
   defaultValue {
    struct=NiagaraExt_VariableValue_DataInterface
    value {dataInterface=&self:NiagaraDataInterfaceSkeletalMesh_1}
   }
  }]
}
