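# Physics asset for the Typhoon Robot backpack cosmetic mesh, defining its collision bodies and the joint between them.
# A large capsule covers the pack bone, and a small capsule sits on the dyn_wire_jnt bone.
# A single constraint ties the wire joint to the pack with tight, equal swing and twist limits, leaving the wire only slight angular play.
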PhysicsAsset /BRCosmetics/Accessories/FORT_Backpacks/Backpack_M_MED_Typhoon_Robot/Meshes/M_MED_Typhoon_Robot_Pack_Physics
bodies {
 dyn_wire_jnt [{shapeType=Capsule, length=4, radius=1.5, center=<-0.0052 0 -0.7961>, rotation=<rot 90 0 0>}]
 pack [{shapeType=Capsule, length=5, radius=10, center=<0 -9.3464 0>}]
}
constraints [{bone1Name=dyn_wire_jnt, bone2Name=pack, swing1LimitDegrees=3, swing2LimitDegrees=3, twistLimitDegrees=3}]
